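# Physics asset for the WartyBrine pickaxe, defining a simulated dangling chain on the weapon mesh.
# Small capsules on the root and chain base anchor four progressively thinner capsules along the dyn_chain bones.
# Constraints link each bone to its parent: the base joint rotates freely, while chain joints lock twist and limit swing to narrow angles.
# The tip joint also locks its second swing axis, bending on one axis only.
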PhysicsAsset /BRCosmetics/Weapons/FORT_Melee/Pickaxe_WartyBrine/Meshes/WartyBrine_Axe_PhysicsAsset
bodies {
 b_AR_Root [{shapeType=Capsule, length=0.505, radius=0.505, shapeName=b_AR_Root_capsule, rotation=<rot 0 -90 0>}]
 base_chain [{shapeType=Capsule, length=0.505, radius=0.505, shapeName=base_chain_capsule, rotation=<rot 0 -90 0>}]
 dyn_chain_01 [{shapeType=Capsule, length=2, radius=1.6, shapeName=dyn_chain_01_capsule, center=<2 0 0>, rotation=<rot 90 0 0>}]
 dyn_chain_02 [{shapeType=Capsule, length=2, radius=1.5, shapeName=dyn_chain_02_capsule, center=<2 0 0>, rotation=<rot 90 0 0>}]
 dyn_chain_03 [{shapeType=Capsule, length=2, radius=1.2, shapeName=dyn_chain_03_capsule, center=<2 0 0>, rotation=<rot 90 0 0>}]
 dyn_chain_04 [{shapeType=Capsule, length=2, radius=0.9, shapeName=dyn_chain_04_capsule, center=<2 0 0>, rotation=<rot 90 0 0>}]
}
constraints [
 {
  bone1Name=base_chain
  bone2Name=b_AR_Root
  swing1LimitDegrees=45
  swing1Motion=Free
  swing2LimitDegrees=45
  swing2Motion=Free
  twistLimitDegrees=45
  twistMotion=Free
 }
 {bone1Name=dyn_chain_01, bone2Name=base_chain, swing1LimitDegrees=25, swing2LimitDegrees=20, twistLimitDegrees=1, twistMotion=Locked}
 {bone1Name=dyn_chain_02, bone2Name=dyn_chain_01, swing1LimitDegrees=25, swing2LimitDegrees=25, twistLimitDegrees=1, twistMotion=Locked}
 {bone1Name=dyn_chain_03, bone2Name=dyn_chain_02, swing1LimitDegrees=25, swing2LimitDegrees=25, twistLimitDegrees=1, twistMotion=Locked}
 {
  bone1Name=dyn_chain_04
  bone2Name=dyn_chain_03
  swing1LimitDegrees=25
  swing2LimitDegrees=1
  swing2Motion=Locked
  twistLimitDegrees=1
  twistMotion=Locked
 }
]
